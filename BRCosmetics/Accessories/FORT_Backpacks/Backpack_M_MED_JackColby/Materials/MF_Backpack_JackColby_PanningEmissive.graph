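# Material function for the JackColby backpack's emissive, adding an animated, audio-reactive color band inside an input mask and a cross decal outside it.
# A time-panned parabolic band along V blends the incoming emissive toward a color parameter.
# Brightness follows frontend-loop audio amplitude, falling back to a slow sine pulse when audio is near silent.
# The cross reuses the matching pickaxe's texture, with adjustable offset and scale, limited by a U-coordinate step.
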MaterialFunction /BRCosmetics/Accessories/FORT_Backpacks/Backpack_M_MED_JackColby/Materials/MF_Backpack_JackColby_PanningEmissive
graph {nodes [
  n1 = Add(
   A=TextureCoordinate()
   B=AppendVector(A=ScalarParameter() {parameterName=Cross Offset U}, B=ScalarParameter() {parameterName=Cross Offset V})
  )
  n2 = ScalarParameter() {defaultValue=1, parameterName=Cross Scale}
  n3 = TextureSample(UVs=Saturate(None=MaterialFunctionCall(UVs=@n1, Texture Scale=@n2) {materialFunction=&ScaleUVsByCenter}.UVs)) {texture=/BRCosmetics/Weapons/FORT_Melee/Pickaxe_JackColby/Textures/T_JackColby_Pickaxe_Cross}
  n4 = ComponentMask(None=TextureCoordinate()) {r=true}
  n5 = Multiply(A=OneMinus(None=Step(Y=@n4) {constX=0.4})) {constB=0.7}
  n6 = FunctionInput() {inputName=EmissiveIn}
  n7 = TextureCoordinate()
  n8 = Constant()
  n9 = Constant() {r=1}
  n10 = Constant() {r=0.01}
  n11 = Saturate(None=MaterialFunctionCall(Input=ComponentMask(None=@n7) {g=true}, Input Low=@n8, Input High=@n9, Target Low=Subtract(A=@n8, B=@n10), Target High=Add(A=@n9, B=@n10)) {materialFunction=&RemapValueRange}.Result)
  n12 = Time()
  n13 = ScalarParameter() {defaultValue=1, parameterName=PanningEmissive_Speed}
  n14 = Frac(None=Add(A=Multiply(A=@n11) {constB=2}, B=Multiply(A=@n12, B=@n13)))
  n15 = OneMinus(None=@n14)
  n16 = Saturate(None=Multiply(A=Multiply(A=@n15, B=@n14)) {constB=4})
  n17 = ScalarParameter() {defaultValue=1, parameterName=PanningEmissive_Power}
  n18 = MaterialFunctionCall!(InAmplitude, InSpectrum, CustomFakeSpectrum, CustomFakeAmplitude) {materialFunction=/BRCosmetics/Global/MaterialFunctions/AudioReactivity/MF_AudioReactivity_FrontendLoop}
  n19 = Saturate(None=Power(Base=@n18.Amplitude) {constExponent=5})
  n20 = Multiply(A=Time(), B=ScalarParameter() {defaultValue=0.5, parameterName=PanningEmissive_PulseSpeed})
  n21 = LinearInterpolate(Alpha=ConstantBiasScale(None=Sine(None=@n20))) {constA=1, constB=3}
  n22 = Multiply(
   A=LinearInterpolate(A=@n6, B=VectorParameter() {parameterName=PanningEmissive_Color, defaultValue=<rgba 0.3207 0 1 1>}.RGB, Alpha=Power(Base=@n16, Exp=@n17))
   B=If(A=@n18.Amplitude, "A > B"=LinearInterpolate(Alpha=@n19) {constA=0.5, constB=5}, "A == B"=@n21, "A < B"=@n21) {constB=0.05}
  )
  n23 = FunctionInput() {inputName=Mask, inputType=FunctionInput_Scalar}
  FunctionOutput(None=LinearInterpolate(A=Multiply(A=@n3.RGB, B=@n5), B=LinearInterpolate(A=@n6, B=@n22, Alpha=@n23), Alpha=@n23))
 ]}
