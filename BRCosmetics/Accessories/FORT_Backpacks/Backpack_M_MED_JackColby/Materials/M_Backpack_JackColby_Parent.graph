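# Parent material for the Jack Colby backpack cosmetic, adding a panning emissive effect to the standard 2023 character shading on skeletal meshes.
# MF_CharacterParent_2023_PreFX builds the base material attributes and supplies a Tech Art Mask.
# MF_Backpack_JackColby_PanningEmissive takes the base emissive and that mask's blue channel, replacing only Emissive Color before MF_CharacterParent_2023_PostFX runs.
# The shading model comes from the material expression graph, and pixel animation is flagged.
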Material /BRCosmetics/Accessories/FORT_Backpacks/Backpack_M_MED_JackColby/Materials/M_Backpack_JackColby_Parent
graph {
 properties {
  bHasPixelAnimation=true
  bUseMaterialAttributes=true
  bUsedWithMeshDeformer=true
  bUsedWithMorphTargets=true
  bUsedWithSkeletalMesh=true
  pixelDepthOffsetMode=PDOM_Legacy
  shadingModel=MSM_FromMaterialExpression
  usedShadingModels=Default Lit | Subsurface | Clear Coat | Subsurface Profile | Cloth
 }
 nodes [
  n1 = MaterialFunctionCall() {materialFunction=&MF_CharacterParent_2023_PreFX}
  n2 = GetMaterialAttributes(None=@n1.Result) {attributeGetTypes [#1]}
  n3 = ComponentMask(None=@n1.Tech Art Mask) {b=true}
  MaterialFunctionCall(
   Material=SetMaterialAttributes(
    MaterialAttributes=@n2.MaterialAttributes
    Emissive Color=MaterialFunctionCall!(EmissiveIn=@n2.Emissive Color, Mask=@n3) {materialFunction=&/MF_Backpack_JackColby_PanningEmissive}.Result
   ) {attributeSetTypes [#1]}
   M=@n1.M
   Pre Skinned Local Position=@n1.M
   Distance Blend=@n1.M
   Fresnel=@n1.M
  ) {materialFunction=&MF_CharacterParent_2023_PostFX}
 ]
}
