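# Physics asset for the M_MED_JackColby backpack cosmetic, covering its nozzle base and four dangling-style "dyn" legs (front and back, left and right).
# Each leg has a tiny capsule on its base bone and a larger angled capsule on its 01 bone.
# Leg bases attach to the shared nozzle base with free swing and twist.
# Each 01 segment hinges on its base, with only swing2 given a nonzero limit.
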PhysicsAsset /BRCosmetics/Accessories/FORT_Backpacks/Backpack_M_MED_JackColby/Meshes/M_MED_JackColby_Pack_Physics
bodies {
 dyn_leg_bk_l_01 [{
   shapeType=Capsule
   length=2.3832
   radius=1.0593
   shapeName=dyn_leg_bk_l_01_capsule
   center=<1.421 0.0295 -1.1279>
   rotation=<rot -0.0308 -90 -44.8113>
  }]
 dyn_leg_bk_l_base [{shapeType=Capsule, length=0.505, radius=0.505, shapeName=dyn_leg_bk_l_base_capsule, rotation=<rot 0 -90 0>}]
 dyn_leg_bk_r_01 [{
   shapeType=Capsule
   length=2.3832
   radius=1.0593
   shapeName=dyn_leg_bk_r_01_capsule
   center=<1.4949 0.0229 1.1279>
   rotation=<rot 0.0308 -90 44.8113>
  }]
 dyn_leg_bk_r_base [{shapeType=Capsule, length=0.505, radius=0.505, shapeName=dyn_leg_bk_r_base_capsule, rotation=<rot 0 -90 0>}]
 dyn_leg_frnt_l_01 [{
   shapeType=Capsule
   length=2.3832
   radius=1.0593
   shapeName=dyn_leg_frnt_l_01_capsule
   center=<1.3829 0.0391 -1.1279>
   rotation=<rot 0.0307 -90 -44.8113>
  }]
 dyn_leg_frnt_l_base [{shapeType=Capsule, length=0.505, radius=0.505, shapeName=dyn_leg_frnt_l_base_capsule, rotation=<rot 0 -90 0>}]
 dyn_leg_frnt_r_01 [{
   shapeType=Capsule
   length=2.3832
   radius=1.0593
   shapeName=dyn_leg_frnt_r_01_capsule
   center=<1.4568 0.0396 1.1279>
   rotation=<rot -0.0307 -90 44.8113>
  }]
 dyn_leg_frnt_r_base [{shapeType=Capsule, length=0.505, radius=0.505, shapeName=dyn_leg_frnt_r_base_capsule, rotation=<rot 0 -90 0>}]
 dyn_nozzle_base [{shapeType=Capsule, length=0.505, radius=0.505, shapeName=dyn_nozzle_base_capsule, rotation=<rot 0 -90 0>}]
}
constraints [
 {
  bone1Name=dyn_leg_bk_l_base
  bone2Name=dyn_nozzle_base
  swing1LimitDegrees=45
  swing1Motion=Free
  swing2LimitDegrees=45
  swing2Motion=Free
  twistLimitDegrees=45
  twistMotion=Free
 }
 {bone1Name=dyn_leg_bk_l_01, bone2Name=dyn_leg_bk_l_base, swing1LimitDegrees=0, swing2LimitDegrees=30, twistLimitDegrees=0}
 {
  bone1Name=dyn_leg_bk_r_base
  bone2Name=dyn_nozzle_base
  swing1LimitDegrees=45
  swing1Motion=Free
  swing2LimitDegrees=45
  swing2Motion=Free
  twistLimitDegrees=45
  twistMotion=Free
 }
 {bone1Name=dyn_leg_bk_r_01, bone2Name=dyn_leg_bk_r_base, swing1LimitDegrees=0, swing2LimitDegrees=30, twistLimitDegrees=0}
 {
  bone1Name=dyn_leg_frnt_l_base
  bone2Name=dyn_nozzle_base
  swing1LimitDegrees=45
  swing1Motion=Free
  swing2LimitDegrees=45
  swing2Motion=Free
  twistLimitDegrees=45
  twistMotion=Free
 }
 {bone1Name=dyn_leg_frnt_l_01, bone2Name=dyn_leg_frnt_l_base, swing1LimitDegrees=0, swing2LimitDegrees=30, twistLimitDegrees=0}
 {
  bone1Name=dyn_leg_frnt_r_base
  bone2Name=dyn_nozzle_base
  swing1LimitDegrees=45
  swing1Motion=Free
  swing2LimitDegrees=45
  swing2Motion=Free
  twistLimitDegrees=45
  twistMotion=Free
 }
 {bone1Name=dyn_leg_frnt_r_01, bone2Name=dyn_leg_frnt_r_base, swing1LimitDegrees=0, swing2LimitDegrees=30, twistLimitDegrees=0}
]
