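# Animation Blueprint for the Jack Colby backpack cosmetic, adding physics and a procedural bounce to the pack.
# The AnimGraph additively offsets the pack bone, runs RigidBody simulation in root-bone space, then translates and scales dyn_buffer.
# The update event drives that bounce and Z scale from a random-frequency sine of game time, gated by a periodic time check.
# It also derives a Parashuting value from owner spine yaw, which the AnimGraph never reads.
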AnimBlueprint /BRCosmetics/Accessories/FORT_Backpacks/Backpack_M_MED_JackColby/Meshes/M_MED_JackColby_Pack_AnimBP
target_skeleton=&/M_MED_JackColby_Pack_Skeleton
graphs [
 AnimGraph {nodes [
   n1 = Input Pose()
   n2 = "Transform (Modify) Bone - Bone: pack"(
    ComponentPose="Transform (Modify) Bone - Bone: pack"(ComponentPose=Local To Component(LocalPose=@n1), Translation="0,0,0", Rotation="0,0,0") {translationMode=BMM_Additive, alphaBoolBlend {blendOption=Linear}, boneToModify {boneName=pack}}
    Translation="-6.000,0.339,7.537"
    Rotation="-10.000,0.000,29.450"
   ) {
    rotationMode=BMM_Additive
    rotationSpace=BCS_BoneSpace
    scaleMode=BMM_Additive
    scaleSpace=BCS_BoneSpace
    translationMode=BMM_Additive
    translationSpace=BCS_BoneSpace
    alphaBoolBlend {blendOption=Linear}
    alphaScaleBiasClamp {bInterpResult=true}
    boneToModify {boneName=pack}
    rotation=<rot -10 0 29.45>
    translation=<-6 0.339 7.537>
   }
   n3 = RigidBody(ComponentPose=@n2) {
    bClampLinearTranslationLimitToRefPose=true
    bOverrideWorldGravity=true
    overridePhysicsAsset=&/M_MED_JackColby_Pack_Physics
    simulationSpace=BaseBoneSpace
    alphaBoolBlend {blendOption=Linear}
    baseBoneRef {boneName=root}
    overrideWorldGravity {z=-980}
    simSpaceSettings {maxAngularAcceleration=350, maxAngularVelocity=250, maxLinearAcceleration=2000, maxLinearVelocity=500, velocityScaleZ=20, worldAlpha=1}
   }
   n4 = "Transform (Modify) Bone - Bone: dyn_buffer"!(
    ComponentPose=@n3
    Alpha=1
    Translation=Get spk_01_bounce() -> (spk_01_bounce:Vector)
    Rotation="0,0,0"
    Scale_X:double=1
    Scale_Y:double=0
    Scale_Z:double=Get scaleZf() -> (scaleZf:double)
   ) -> (Pose) {
    rotationSpace=BCS_ParentBoneSpace
    scaleMode=BMM_Additive
    scaleSpace=BCS_BoneSpace
    translationMode=BMM_Additive
    translationSpace=BCS_BoneSpace
    alphaBoolBlend {blendOption=Linear}
    boneToModify {boneName=dyn_buffer}
   }
   Output Pose(Result=Component To Local(ComponentPose=@n4)) {layerGroup=None}
  ]}
 EventGraph {nodes [
   Event BlueprintUpdateAnimation()
   n1 = GetDeltaTransformFromRefPose!(self=GetOwnerSkeletalMeshComponent(), BoneName=spine_05, BaseName=spine_01) -> (ReturnValue_Location:Vector, ReturnValue_Rotation:Rotator, ReturnValue_Scale:Vector)
   > Set Parashuting(Parashuting:double=MapRangeClamped(Value=BreakRotator(InRot=@n1.ReturnValue_Rotation).Yaw, InRangeA=-25, InRangeB=-65)) -> (Output_Get:double)
   > Set DeltaTimeLocal(DeltaTimeLocal:double=0) -> (Output_Get:double)
   n2 = GetGameTimeInSeconds()
   > Branch(Condition="Nearly Equal (Float)"(A="% (Float)"!(A:double=@n2, B:double=0.43) -> (ReturnValue:double), B=0, ErrorTolerance=0.26))
   n3 = float * float(
    A=RandomFloatFromStream!(Stream:"Random Stream Structure (by ref)"=Get random_stream() -> (random_stream:Random Stream Structure)) -> (ReturnValue:float)
    B=0.9
   )
   n4 = "Sin (Radians)"(A=float * float(A="Clamp (Float)"(Value=@n3, Min=0.5, Max=0.8), B=GetGameTimeInSeconds()))
   > Set spk_01_bounce(spk_01_bounce:Vector=MakeVector(X=float * float(A=@n4, B=3), Y=0)) -> (Output_Get:Vector)
   > Set scaleZf(scaleZf:double=MapRangeClamped(Value="Absolute (Float)"(A=@n4), InRangeA=0, InRangeB=1, OutRangeA=0.75, OutRangeB=1.75)) -> (Output_Get:double)
   > Set scaleYf(scaleYf:double=1) -> (Output_Get:double)
   TryGetPawnOwner()
  ]}
]
variables [
 bowl_bounce {type=Rotator}
 DeltaTimeLocal {type=number}
 Parashuting {type=number}
 random_stream {type=RandomStream, default {initialSeed=0, seed=0}}
 scaleYf {type=number}
 scaleZf {type=number}
 spk_01_bounce {type=Vector}
 spk_02_bounce {type=Vector}
]
